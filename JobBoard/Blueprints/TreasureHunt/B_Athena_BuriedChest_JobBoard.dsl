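; Buried chest Blueprint for the Job Board treasure hunt that drops extra gold when looted.
; On server damage it casts the instigator to a Fort player controller and stores it as the last damage instigator.
; On loot it loops to spawn gold pickups, with drop count and amount per drop read from scalable floats.
; Pickups appear raised and offset along the forward vector rotated a quarter turn about Z, and are tossed that way.
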(event EventBeginPlay
  (|CallParentFunction))

(event EventOnDamageServer
  (params (Damage :type Float%20%28single-precision%29) (DamageTags :type Gameplay%20Tag%20Container%20Structure%20%28by%20ref%29) (Momentum :type Vector :default "0, 0, 0") (HitInfo :type Hit%20Result%20Structure%20%28by%20ref%29) (InstigatedBy :type Controller%20Object%20Reference) (DamageCauser :type Actor%20Object%20Reference) (EffectContext :type Gameplay%20Effect%20Context%20Handle%20Structure))
  (bind _asfort_player_controller (Utilities|Casting|CastToFortPlayerController InstigatedBy))
  (Variables|Default|SetLastDamageInstigator _asfort_player_controller))

(event EventOnLoot
  (for _ (range 1 (+ (ScalableFloat|GetValueasInteger (Variables|Default|GetHowManyTimesToDropGold)) 1))
    (Inventory|SpawnItemVariantPickupinWorld (Utilities|Struct|MakeSpawnItemVariantParams "/Game/Items/ResourcePickups/Athena_WadsItemData.Athena_WadsItemData" "(GameplayTags=)" (ScalableFloat|GetValueasInteger (Variables|Default|GetHowMuchGoldToDropEachTime)) (+ (+ (* (Math|Vector|RotateVectorAroundAxis (Transformation|GetActorForwardVector) 9e+01 "0.000000,0.000000,1.000000") 5e+01) (Transformation|GetActorLocation)) "0.000000,0.000000,50.000000") (* (Math|Vector|RotateVectorAroundAxis (Transformation|GetActorForwardVector) 9e+01 "0.000000,0.000000,1.000000") 2e+02) 0 true true false -1 "Container" "Unset" 0 0 false))))

(fn ConstructionScript
  (|CallParentFunction))
