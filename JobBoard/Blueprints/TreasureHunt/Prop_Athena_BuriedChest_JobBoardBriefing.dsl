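; Buried chest prop for Job Board treasure hunt quests, shown to the owning player's squad by a server-made tracking radius.
; Digging it up hides dirt and FX and plays a gameplay cue.
; Opening clears tracking, broadcasting a verb to the squad for another team's poach or a same-team destruction of an unsearched chest.
; Squad quest completion or urgent-quest expiry sets bQuestCompleted, whose OnRep removes tracking and unbinds chest events.
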(event EventBeginPlay
  (bind _returnvalue (Actor|GetOwner))
  ;[0 0 0] Begin play - get the owner of the spawned prop and bind events to when the buried chest is searched or destroyed
  (|CallParentFunction)
  (Utilities|IsValid _returnvalue
    (:Is%20Valid
      (bind _fortpawn (Class|FortQuestBlueprintLibrary|GetFortPawnFromObject _returnvalue)
        (:Valid
          (bind _asfort_player_pawn (Utilities|Casting|CastToFortPlayerPawn _fortpawn))
          (Variables|Default|SetOwningPlayerPawn _asfort_player_pawn)
          (bind _squadplayercontrollers (CallFunction|GetOwnerSquadmateControllers))
          (Variables|Default|SetSquadPlayerControllers _squadplayercontrollers)
          (CallFunction|MakeTrackingRadiusforChest 0)
          (CallFunction|TrackQuestStateforSquad))))))

(event Custom|OnBuriedChestDestroyed
  (params (DestroyedActor :type Actor%20Object%20Reference))
  ;[0 0 0] On buried chest destroyed or searched
  (Variables|Default|SetbChestDestroyed true)
  ;;[0.1094 0.1094 0.1094] Get the last instigator to damage the chest to check if it was a player on the same team as the owner that broke it open
  (bind _asb_athena_buried_chest_job_board (Utilities|Casting|CastToB_Athena_BuriedChest_JobBoard (Variables|Default|GetChestObject)))
  (Utilities|IsValid (Variables|/JobBoard/Blueprints/TreasureHunt/B_Athena_BuriedChest_JobBoard.B_Athena_BuriedChest_JobBoard_C|GetLastDamageInstigator _asb_athena_buried_chest_job_board)
    (:Is%20Valid
      (CallFunction|HandleChestOpened (Game|Player|GetPlayerPawn (Variables|/JobBoard/Blueprints/TreasureHunt/B_Athena_BuriedChest_JobBoard.B_Athena_BuriedChest_JobBoard_C|GetLastDamageInstigator _asb_athena_buried_chest_job_board))))))

(event Custom|OnBuriedChestSearched
  (params (SelfActor :type Building%20Actor%20Object%20Reference) (InteractingPawn :type Fort%20Pawn%20Object%20Reference))
  ;[0 0 0] On buried chest destroyed or searched
  (CallFunction|HandleChestOpened InteractingPawn))

(event EventOnMaxDug
  (bind _niagara (Variables|Default|GetNiagara))
  (bind _dirt (Variables|Default|GetDirt))
  ;[0 0 0] When the buried chest is dug up, remove the FX
  (|CallParentFunction)
  (Default|CallOnMaxDugDispatch self)
  (Rendering|SetVisibility _niagara false false)
  (Development|SetHiddeninGame _niagara true false)
  (Rendering|SetVisibility _dirt false false)
  (Development|SetHiddeninGame _dirt true false)
  ;[1 1 1] 39.00 - added a GCN to play when the chest is dug up
  (GameplayCue|ExecuteGameplayCueOnActor(Burst) (Variables|Default|GetChestObject) (Variables|Default|GetGC_DugUp) (Ability|GameplayCue|MakeGameplayCueParameters 0.0)))

(event Custom|TrackQuestStateForSquad
  ;[0 0 0] If anyone on the squad completes the quest (will trigger if failed as well) then unbind events from the chest and remove tracking radius
  (Networking|SwitchHasAuthority
    (:Authority
      (for _array_element (Variables|Default|GetSquadPlayerControllers)
        (bind _returnvalue (EventProcessorSubsystem|StartListeningToFilteredEvent 0 (Game|Player|GetPlayerPawn _array_element) (Variables|Default|GetQuestCompletedFilter) true (delegate Custom|QuestCompleted)))
        (Utilities|Array|AddUnique (Variables|Default|GetQuestCompletedFilterArray) _returnvalue)
        ;; NOTE - Transient Quests Component is likely to be deprecated in the future (TBD timing). Need a new way to detect urgent quest expiring in BP
        (EventDispatchers|BindEventtoOnUrgentQuestTimeExpired (Class|FortPlayerPawn|GetTransientQuestsComponent (Game|Player|GetPlayerPawn _array_element)) (delegate Custom|TimeExpired))))))

(event Custom|TimeExpired
  (params (UrgentQuestData :type Urgent%20Quest%20Data%20Structure%20%28by%20ref%29))
  ;[0 0 0] If anyone on the squad completes the quest (will trigger if failed as well) then unbind events from the chest and remove tracking radius
  ;; NOTE - Transient Quests Component is likely to be deprecated in the future (TBD timing). Need a new way to detect urgent quest expiring in BP
  (Variables|Default|SetbQuestCompleted true))

(event Custom|QuestCompleted
  ;[0 0 0] If anyone on the squad completes the quest (will trigger if failed as well) then unbind events from the chest and remove tracking radius
  (Variables|Default|SetbQuestCompleted true)
  (for _array_element (Variables|Default|GetQuestCompletedFilterArray)
    (EventProcessorSubsystem|StopListeningToFilteredEvent _array_element)))

(event EventEndPlay
  (params (EndPlayReason :type EEndPlayReason%20Enum :default "Destroyed"))
  ;[0 0 0] End play - if events were never unbound from the chest object, unbind them
  (|CallParentFunction EndPlayReason)
  (Utilities|IsValid (Variables|Default|GetChestObject)
    (:Is%20Valid
      (Game|UnbindEventfromOnDestroyed (Variables|Default|GetChestObject) (EventDispatchers|CreateEvent :!event Custom|OnBuriedChestDestroyed))
      (Interaction|UnbindEventfromOnInteract (Variables|Default|GetChestObject) (EventDispatchers|CreateEvent :!event Custom|OnBuriedChestSearched))
      (for _array_element (Variables|Default|GetSquadPlayerControllers)
        (EventDispatchers|UnbindEventfromOnUrgentQuestTimeExpired (Class|FortPlayerPawn|GetTransientQuestsComponent (Game|Player|GetPlayerPawn _array_element)) (EventDispatchers|CreateEvent :!event Custom|TimeExpired))))))

(fn BroadcastVerbToSquad
  (params (VerbName :type Gameplay%20Tag%20Structure) (SubjectTags :type Gameplay%20Tag%20Container%20Structure) (Amount :type Integer))
  (for _array_element (Variables|Default|GetSquadPlayerControllers)
    (Utilities|IsValid _array_element
      (:Is%20Valid
        (bind _result (CustomVerb|AddTagsforSubject (CustomVerb|MakeCustomVerbMessage (Game|Player|GetPlayerPawn _array_element) VerbName Amount) "Job Type" SubjectTags))
        (CustomVerb|BroadcastCustomVerbMessage _result)))))

(fn ChestObjectSet
  (|CallParentFunction)
  (Game|BindEventtoOnDestroyed (Variables|Default|GetChestObject) (EventDispatchers|CreateEvent :!event OnBuriedChestDestroyed))
  (Interaction|BindEventtoOnInteract (Variables|Default|GetChestObject) (EventDispatchers|CreateEvent :!event OnBuriedChestSearched)))

(fn GetOwnerSquadmateControllers
  (bind _owningplayerpawn (Variables|Default|GetOwningPlayerPawn))
  ;[0 0 0] Get the controllers for the owning player's squadmates. Includes self
  (bind _returnvalue (Fortnite|Utilities|AttemptGetPlayerStateforPawn _owningplayerpawn))
  (bind _asfort_player_state_athena (Utilities|Casting|CastToFortPlayerStateAthena _returnvalue))
  (bind _returnvalue_2 (Class|FortGameStateAthena|GetSquadMemberPlayerStates (Fortnite|Athena|GameState|GetGameStateAthenafromContext _owningplayerpawn) _asfort_player_state_athena true))
  (for _array_element _returnvalue_2
    (Utilities|Array|AddUnique (LocalVariables|GetSquadmateControllers) (Fortnite|GetFortPlayerControllerfromActor _array_element)))
  (return (LocalVariables|GetSquadmateControllers)))

(fn HandleChestOpened
  (params (InstigatingPawn :type Fort%20Pawn%20Object%20Reference))
  ;[0 0 0] On chest opened or destroyed - send quest completion or poached progress
  (CallFunction|RemoveTrackingRadiusforChest 0)
  (if (Team|OnSameTeam InstigatingPawn (Variables|Default|GetOwningPlayerPawn))
    (if (Variables|Default|GetbChestDestroyed)
      ;;[0.1094 0.1094 0.1094] FORT-701797 - ensure that players can't get double credit for the same chest by destroying one that's already open
      (if (not (Container|IsSearched (Variables|Default|GetChestObject)))
        (CallFunction|BroadcastVerbtoSquad "(TagName=\"CustomVerb.JobBoard.DestroyedChest\")" (Variables|Default|GetJobTypeTags) 1)
        (else
          _))
      (Game|UnbindEventfromOnDestroyed (Variables|Default|GetChestObject) (EventDispatchers|CreateEvent :!event OnBuriedChestDestroyed))
      (Interaction|UnbindEventfromOnInteract (Variables|Default|GetChestObject) (EventDispatchers|CreateEvent :!event OnBuriedChestSearched))
      (else
        _))
    (else
      (CallFunction|BroadcastVerbtoSquad "(TagName=\"CustomVerb.JobBoard.Poached\")" (Variables|Default|GetJobTypeTags) (Variables|Default|GetQuestObjectiveCountToComplete)))))

(fn MakeTrackingRadiusForChest
  (params (ThisPawnOnly :type Fort%20Player%20Pawn%20Object%20Reference))
  ;[0 0 0] Makes a radial tracking component around the buried chest
  (Networking|SwitchHasAuthority
    (:Authority
      (bind _returnvalue (RadiusTracker|MakeRadiusTrackingInitializer (ScalableFloat|GetValueatLevel (LocalVariables|GetTrackingMaxCenterOffset)) (ScalableFloat|GetValueatLevel (LocalVariables|GetTrackingMinRadius)) (ScalableFloat|GetValueatLevel (LocalVariables|GetTrackingMaxRadius)) (ScalableFloat|GetValueatLevel (LocalVariables|GetTrackingTimeBetweenUpdates)) (ScalableFloat|GetValueatLevel (LocalVariables|GetTrackingRadiusShrinkAmount)) "(R=1.000000,G=0.456411,B=0.000000,A=0.400000)" 0 self (Variables|Default|GetSquadPlayerControllers)))
      (Fortnite|AddActorTrackingInfo (RadiusTracker|GetGameStateRadiusTrackerComponent) _returnvalue))))

(fn OnRep_bQuestCompleted
  (CallFunction|RemoveTrackingRadiusforChest 0)
  (Game|UnbindEventfromOnDestroyed (Variables|Default|GetChestObject) (EventDispatchers|CreateEvent :!event OnBuriedChestDestroyed))
  (Interaction|UnbindEventfromOnInteract (Variables|Default|GetChestObject) (EventDispatchers|CreateEvent :!event OnBuriedChestSearched)))

(fn Prop_Athena_BuriedChest_JobBoardBriefing_AutoGenFunc
  (params (DestroyedActor :type Actor%20Object%20Reference)))

(fn RemoveTrackingRadiusForChest
  (params (ThisPawnOnly :type Fort%20Player%20Pawn%20Object%20Reference))
  (bind _returnvalue (Fortnite|GetFortPlayerControllerfromActor ThisPawnOnly))
  (Networking|SwitchHasAuthority
    (:Authority
      (Utilities|IsValid ThisPawnOnly
        (:Is%20Valid
          (Utilities|IsValid _returnvalue
            (:Is%20Valid
              (RadiusTracker|RemoveTrackingforActorfromPlayer (RadiusTracker|GetGameStateRadiusTrackerComponent) self _returnvalue "Team"))))
        (:Is%20Not%20Valid
          (for _array_element (Variables|Default|GetSquadPlayerControllers)
            (Utilities|IsValid _array_element
              (:Is%20Valid
                (RadiusTracker|RemoveTrackingforActorfromPlayer (RadiusTracker|GetGameStateRadiusTrackerComponent) self _array_element "Team")))))))))

(fn ConstructionScript
  (|CallParentFunction))
